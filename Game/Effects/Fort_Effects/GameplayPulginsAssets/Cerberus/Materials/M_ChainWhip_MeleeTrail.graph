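# Additive unlit Niagara ribbon material, apparently for Cerberus chain whip melee trails.
# Dynamic parameters place each trail section within a gradient split by NumOfAnimTrailSections, which scales particle color and remaps texture UVs.
# Panned noise and plasma textures combine into a sharpened mask, shaped by length gradients, particle alpha and optional edge fade.
# View-angle fade applies on all shading paths, while depth fade skips mobile.
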Material /Game/Effects/Fort_Effects/GameplayPulginsAssets/Cerberus/Materials/M_ChainWhip_MeleeTrail
graph {
 properties {bUsedWithNiagaraRibbons=true, blendMode=BLEND_Additive, pixelDepthOffsetMode=PDOM_Legacy, shadingModel=MSM_Unlit, usedShadingModels=Unlit}
 nodes [
  Constant() {r=1}
  n1 = ParticleColor()
  n2 = MaterialFunctionCall(Float2=TextureCoordinate()) {materialFunction=&BreakOutFloat2Components}
  n3 = ScalarParameter() {defaultValue=1, parameterName=NumOfAnimTrailSections}
  n4 = DynamicParameter() {defaultValue=<rgba 0 1 15 1>, paramNames [IndexOfPartAnimTrails, Tile, Offset, EdgeFade]}
  n5 = Add(A=Divide(A=@n2.G, B=@n3) {constB=3}, B=Divide(A=@n4.IndexOfPartAnimTrails, B=@n3) {constB=3})
  MaterialFunctionCall(Emissive=Multiply(A=@n1.RGB, B=@n5)) {materialFunction=&MF_FX_EmissiveAdjustments}
  n6 = Multiply(A=OneMinus(None=@n2.R), B=@n5)
  n7 = Multiply(A=Constant() {r=0.2}, B=@n4.Tile)
  n8 = Constant() {r=1}
  n9 = Panner(
   Coordinate=MaterialFunctionCall(R=Multiply(A=@n2.R, B=@n7), G=Multiply(A=@n5, B=@n8)) {materialFunction=&AppendMany}.RG
   Time=Add(A=Time(), B=@n4.Offset)
  ) {speedX=-0.5}
  n10 = TextureSample(UVs=@n9) {samplerType=SAMPLERTYPE_LinearColor, texture=&T_NoisePack_LinearColor}
  n11 = TextureSample(UVs=@n9) {samplerType=SAMPLERTYPE_LinearGrayscale, texture=/Game/Effects/Fort_Effects/Emotes/UltraLight/Texture/T_Plasma_A_Greyscale}
  n12 = Power(Base=Multiply(A=Add(A=@n10.B, B=@n11.R)) {constB=0.7}) {constExponent=3}
  n13 = TextureCoordinate()
  n14 = Multiply(A=OneMinus(None=MaterialFunctionCall(Float2=@n13) {materialFunction=&BreakOutFloat2Components}.G)) {constB=12}
  n15 = Multiply(A=Multiply(A=@n1.A, B=Multiply(A=@n6, B=@n12)), B=LinearInterpolate(B=Saturate(None=@n14), Alpha=@n4.EdgeFade) {constA=1})
  n16 = DepthFade(Opacity=@n15)
  Multiply(
   A=ShadingPathSwitch(Default=@n16, Deferred=@n16, Mobile=@n15)
   B=MaterialFunctionCall("FadeSensitivity (optional)"=@n8) {materialFunction=&FadeBasedOnViewAngle}.Result
  )
 ]
}
